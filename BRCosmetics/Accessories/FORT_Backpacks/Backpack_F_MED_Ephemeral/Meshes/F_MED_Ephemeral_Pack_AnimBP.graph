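# Animation Blueprint for the Ephemeral backpack that adjusts the pack's pose while the character is skydive floating.
# Each update, the event graph converts the bIsSkydiveFloating flag to a float and stores it as an alpha, with no smoothing in the graph.
# The anim graph feeds that alpha into a Modify Bone node that additively offsets and rolls the pack bone in bone space on top of the input pose.
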AnimBlueprint /BRCosmetics/Accessories/FORT_Backpacks/Backpack_F_MED_Ephemeral/Meshes/F_MED_Ephemeral_Pack_AnimBP
target_skeleton=&/F_MED_Ephemeral_Pack_Skeleton
graphs [
 AnimGraph {nodes [
   n1 = "Transform (Modify) Bone - Bone: pack"(
    ComponentPose=Local To Component(LocalPose=Input Pose())
    Alpha=Get Is Skydive Floating Alpha() -> (Is Skydive Floating Alpha:double)
    Translation="0,-1.5,0"
    Rotation="0,0,20"
   ) {
    rotationMode=BMM_Additive
    rotationSpace=BCS_BoneSpace
    translationMode=BMM_Additive
    translationSpace=BCS_BoneSpace
    alphaBoolBlend {blendOption=Linear}
    boneToModify {boneName=pack}
    rotation {roll=20}
    translation {y=-1.5}
   }
   Output Pose(Result=Component To Local(ComponentPose=@n1)) {layerGroup=None}
  ]}
 EventGraph {nodes [
   Event BlueprintUpdateAnimation()
   > Sequence()
   > Set Is Skydive Floating Alpha(Is Skydive Floating Alpha:double="To Float (Boolean)"(InBool=Get bIsSkydiveFloating())) -> (Output_Get:double)
   TryGetPawnOwner()
  ]}
]
variables [Is Skydive Floating Alpha {type=number}]
